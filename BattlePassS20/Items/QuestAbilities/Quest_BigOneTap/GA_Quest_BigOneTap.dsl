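; Gameplay ability for the BigOneTap battle pass quest, crediting a single large event against a player target, presumably a heavy one-shot hit.
; On activation from a gameplay event it calls the parent, then checks that the target tags include the Athena player pawn and the magnitude is at least 100.
; A match sends a BigOneTap custom stat event through the owning player's quest manager, and the ability ends either way.
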(event Ability|EventActivateAbilityFromEvent
  (params (EventData :type Gameplay%20Event%20Data%20Structure%20%28by%20ref%29))
  (bind (_eventtag _instigator _target _optionalobject _optionalobject2 _contexthandle _instigatortags _targettags _eventmagnitude) (Utilities|Struct|BreakGameplayEventData EventData))
  (|CallParentFunction EventData)
  (if (and (GameplayTags|HasTag _targettags "(TagName=\"Pawn.Athena.Player\")") (>= _eventmagnitude 1e+02))
    (Quest|SendComplexCustomStatEvent (Variables|GAVars|Data|GetOwningPlayerQuestManager) 0 "(GameplayTags=((TagName=\"Athena.Quests.BigOneTap\")))" "(GameplayTags=)" 1)
    (Ability|EndAbility)
    (else
      (Ability|EndAbility))))
